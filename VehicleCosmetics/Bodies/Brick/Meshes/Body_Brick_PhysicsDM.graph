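# Physics asset for the Brick vehicle body cosmetic, giving its mesh a simplified collision and physics shape.
# It holds a single body on the root bone with one box shape, offset forward and upward from the bone origin.
# With no constraints, the body simulates as one rigid piece with no articulated parts.
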PhysicsAsset /VehicleCosmetics/Bodies/Brick/Meshes/Body_Brick_PhysicsDM
bodies {root [{shapeType=Box, extentX=444, extentY=216, extentZ=116, shapeName=root_box, center=<24.1819 0 100.5541>}]}
constraints []
